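# Niagara system for the TurboTechno vehicle boost cosmetic, drawing exhaust trails and bursts at boost ports on a skeletal mesh.
# A ribbon emitter layers five ribbon renderers with per-layer colors and shrinking secondary widths.
# A second emitter partitions short-lived particles across port sockets, driving two mesh renderers and a sprite renderer with curve-animated scale.
# User flags switch colors and intensities for painted variants, and velocity and contrast for frontend preview.
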NiagaraSystem /VehicleCosmetics/Boosts/Boost_TurboTechno/FX/NS_Boost_TurboTechno
dependencies {
 usedDataInterfaces [
  &NiagaraDataInterfaceColorCurve
  &NiagaraDataInterfaceCurve
  &NiagaraDataInterfaceArrayFloat
  &NiagaraDataInterfaceVectorCurve
  &NiagaraDataInterfaceSkeletalMesh
 ]
 usedDynamicInputs []
 usedModules [
  &self:SystemSpawnScript.SystemScriptSource.SystemScriptGraph.NiagaraNodeAssignment_0.SetVariables_#1
  &SystemState
 ]
}
emitters {
 NE_VC_Boost_Ribbon_Parent001 {
  emitterSpawnScript [/SetVariables(
    Emitter.FrontContrast:NiagaraFloat=&MakeCustomFloatFromBool(Bool=$User.bIsFrontendPreview, True Float=2, False Float=1.0625)
    Emitter.R1 MultParam:NiagaraFloat=&MakeCustomFloatFromBool(Bool=$User.bIsPainted, True Float=5, False Float=2)
    Emitter.R2 MultParam:NiagaraFloat=&MakeCustomFloatFromBool(Bool=$User.bIsPainted, True Float=5, False Float=1.5)
    Emitter.R3 MultParam:NiagaraFloat=&MakeCustomFloatFromBool(Bool=$User.bIsPainted, True Float=16, False Float=8)
    Emitter.Size Factor=8
    Emitter.WPO Mask Intensity:NiagaraFloat=&MakeCustomFloatFromBool(Bool=$User.bIsFrontendPreview, True Float=0.4, False Float=0.2)
   ) {isSetParameters=true}]
  emitterUpdateScript [
   &EmitterState()
   &VC_FilteredSocketCount()
   &VC_SpawnRibbon_v1(Number of Ribbons=$Emitter.PortCount, SpawnRate=60)
  ]
  particleSpawnScript [
   &VC_PaintedVariantsOverride(Black=<rgba 0.4271 0.1579 0.1579 1>) {enabled=false}
   /SetVariables(
    Particles.R2 Size:NiagaraFloat=&Multiply_Float(A=$Emitter.Size Factor, B=21)
    Particles.R3 Size:NiagaraFloat=&Multiply_Float(A=$Emitter.Size Factor, B=6)
    Particles.R4 Size=4
   ) {isSetParameters=true}
   /VehicleCosmetics/FX/Niagara/Modules/MergeEmitters/VC_ExtraBindingColors(
    ExtraColors=4
    R1 Color:LinearColor=&MakeCustomLinearColorFromBool(
     Bool=$User.bIsPainted
     True Color=$User.PaintedColor
     False Color=&ColorFromCurve(ColorCurve=DataInterface() {propertyValues {
        AlphaCurve {keys [<at 0 0.5 RCIM_Cubic>,<at 1 0.5 RCIM_Cubic>]}
        BlueCurve {keys [<at 0 1 RCIM_Cubic>,<at 1 1 RCIM_Cubic>]}
        GreenCurve {keys [<at 0 0>,<at 1 0.5>]}
        RedCurve {keys [<at 0 0.4>,<at 1 1>]}
       }})
    )
    R2 Color:LinearColor=&MakeCustomLinearColorFromBool(Bool=$User.bIsPainted, True Color=$User.PaintedColor, False Color=<rgba 1 0.2 1 1>)
    R3 Color:LinearColor=&MakeCustomLinearColorFromBool(Bool=$User.bIsPainted, True Color=$User.PaintedColor, False Color=<rgba 0.3 0.4714 1.5 1>)
    R4 Color:LinearColor=&MakeCustomLinearColorFromBool(Bool=$User.bIsPainted, True Color=$User.PaintedColor, False Color=<rgba 2 2.5 5 1>)
   )
   &VC_PaintedVariantsOverride(
    R1 Black:LinearColor=<rgba 0 0 0 1>
    R1 Crimson:LinearColor=<rgba 0.232 0 0.0063 1>
    R1 DefaultColor=&MakeCustomLinearColorFromBool(
     Bool=$User.bIsPainted
     True Color=$User.PaintedColor
     False Color=&ColorFromCurve(ColorCurve=DataInterface() {propertyValues {
        AlphaCurve {keys [<at 0 0.5 RCIM_Cubic>,<at 1 0.5 RCIM_Cubic>]}
        BlueCurve {keys [<at 0 1 RCIM_Cubic>,<at 1 1 RCIM_Cubic>]}
        GreenCurve {keys [<at 0 0>,<at 1 0.5>]}
        RedCurve {keys [<at 0 0.4>,<at 1 1>]}
       }})
    )
    R1 ForestGreen:LinearColor=<rgba 0 0.16 0.0081 1>
    R1 Grey:LinearColor=<rgba 0.1 0.1 0.1 1>
    R1 Lime:LinearColor=<rgba 0.135 0.27 0 1>
    R1 Orange:LinearColor=<rgba 0.43 0.1075 0 1>
    R1 Pink:LinearColor=<rgba 0.44 0.088 0.312 1>
    R1 Purple:LinearColor=<rgba 0.1 0 0.6 1>
    R1 Saffron:LinearColor=<rgba 0.38 0.2913 0 1>
    R1 SkyBlue:LinearColor=<rgba 0 0.2662 0.4 1>
    R1 TitaniumWhite:LinearColor=<rgba 0.4 0.4 0.4 1>
    R2 Black:LinearColor=<rgba 0 0 0 1>
    R2 Crimson:LinearColor=<rgba 0.232 0 0.0063 1>
    R2 DefaultColor:LinearColor=&MakeCustomLinearColorFromBool(Bool=$User.bIsPainted, True Color=$User.PaintedColor, False Color=<rgba 1 0.2 1 1>)
    R2 ForestGreen:LinearColor=<rgba 0 0.16 0.0081 1>
    R2 Grey:LinearColor=<rgba 0.1 0.1 0.1 1>
    R2 Lime:LinearColor=<rgba 0.135 0.27 0 1>
    R2 Orange:LinearColor=<rgba 0.27 0.041 0.0108 1>
    R2 Pink:LinearColor=<rgba 0.2133 0.08 0.4 1>
    R2 Purple:LinearColor=<rgba 0.1 0 0.6 1>
    R2 Saffron:LinearColor=<rgba 0.15 0.15 0 1>
    R2 SkyBlue:LinearColor=<rgba 0 0.1667 0.5 1>
    R2 TitaniumWhite:LinearColor=<rgba 0.2 0.225 0.25 1>
    R3 Black:LinearColor=<rgba 0.05 0.05 0.05 1>
    R3 BurntSienna:LinearColor=<rgba 0.25 0.0527 0.021 1.25>
    R3 Cobalt:LinearColor=<rgba 0.05 0.1292 1.25 1.25>
    R3 Crimson:LinearColor=<rgba 0.664 0.0664 0.0826 1.25>
    R3 DefaultColor:LinearColor=&MakeCustomLinearColorFromBool(Bool=$User.bIsPainted, True Color=$User.PaintedColor, False Color=<rgba 0.3 0.4714 1.5 1>)
    R3 ForestGreen:LinearColor=<rgba 0.0443 0.3 0.021 1.25>
    R3 Grey:LinearColor=<rgba 0.28 0.28 0.28 1.25>
    R3 Lime:LinearColor=<rgba 0.4083 0.7 0 1.25>
    R3 Orange:LinearColor=<rgba 1 0.25 0 1.25>
    R3 Pink:LinearColor=<rgba 1 0.3 0.7455 1.35>
    R3 Purple:LinearColor=<rgba 0.5 0 1 1.25>
    R3 Saffron:LinearColor=<rgba 1 1 0.2 1>
    R3 SkyBlue:LinearColor=<rgba 0.1 0.55 1 1.25>
    R3 TitaniumWhite:LinearColor=<rgba 1.08 1.16 1.2 1>
    R4 Black:LinearColor=<rgba 0.01 0.01 0.01 8>
    R4 BurntSienna:LinearColor=<rgba 0.4 0.1067 0.08 6>
    R4 Cobalt:LinearColor=<rgba 0.24 0.2924 2 4>
    R4 Crimson:LinearColor=<rgba 0.864 0 0.0234 4>
    R4 ForestGreen:LinearColor=<rgba 0 0.8 0 4>
    R4 Grey:LinearColor=<rgba 0.5 0.5 0.5 4>
    R4 Lime:LinearColor=<rgba 0.3945 0.789 0 4>
    R4 Orange:LinearColor=<rgba 1.15 0.3258 0 4>
    R4 Pink:LinearColor=<rgba 2 0.4 1.4182 4>
    R4 Purple:LinearColor=<rgba 0.325 0 0.78 8>
    R4 Saffron:LinearColor=<rgba 1 0.9183 0.02 4>
    R4 SkyBlue:LinearColor=<rgba 0 0.6654 1 4>
    R4 TitaniumWhite:LinearColor=<rgba 1 1 1 4>
    Use R1 Painted Overrides=true
    Use R2 Painted Overrides=true
    Use R3 Painted Overrides=true
    Use R4 Painted Overrides=true
   ) {module=VC_PaintedVariantsOverride001}
   &VC_InitializeRibbon_v1(
    Enable Skeletal Position=true
    Lifetime=0.4
    Ribbon Width:NiagaraFloat=&Multiply_Float(A=$Emitter.Size Factor, B=10)
   )
   &AddVelocity(Velocity=&MakeCustomVectorFromBool(Bool=$User.bIsFrontendPreview, True Vector=<-3000 0 0>, False Vector=<-6000 0 0>)) {module=AddVelocity001}
   &ShapeLocation(MaskSpawnGroup=false, Override Seed=false, Sphere Radius=0.5, UseOwnerScale=true) {enabled=false}
  ]
  particleUpdateScript [
   &ParticleState()
   &InheritSourceMovement(ApplyForce=false, ApplyPosition=true, ApplyPositionFalloff=false)
   &SolveForcesAndVelocity(Clamp Velocity=false, Limit Acceleration=false)
   &VC_UpdateRibbon_v1(Enable Skeletal Position=true)
   /VehicleCosmetics/FX/Niagara/Modules/MergeEmitters/VC_ScaleExtraColors(
    ExtraColors=3
    R1 Scale Alpha:NiagaraFloat=&FloatFromCurve(FloatCurve=DataInterface() {propertyValues {Curve {keys [<at 0 1>,<at 0.1 0.6>,<at 1 0.2>]}}})
    R2 Scale Alpha:NiagaraFloat=&FloatFromCurve(FloatCurve=DataInterface() {propertyValues {Curve {keys [<at 0 1>,<at 0.1 0.6>,<at 1 0.2>]}}})
   ) {enabled=false}
   /SetVariables(
    Particles.R2 Size:NiagaraFloat=&Multiply_Float(
     A=$Particles.Initial.R2 Size
     B=&FloatFromCurve(FloatCurve=DataInterface() {propertyValues {Curve {keys [<at 0 0.8>,<at 1 0.2>]}}})
    )
    Particles.R3 Size:NiagaraFloat=&Multiply_Float(
     A=$Particles.Initial.R3 Size
     B=&FloatFromCurve(FloatCurve=DataInterface() {propertyValues {Curve {keys [<at 0 0.6 RCIM_Cubic>, {arriveTangent=-0.4, interpMode=RCIM_Cubic, leaveTangent=-0.4, time=0.3, value=1}, <at 1 0.2>]}}})
    )
   ) {isSetParameters=true}
   &DynamicMaterialParameters(
    Index 3 Param 3=&SelectFloatFromArray(Direct Array Index=&SpawnGroup(), Float Selection Array=DataInterface() {propertyValues {FloatData [0,1]}})
    Module.Index3Float0Write=false
    Module.Index3Float1Write=false
    Module.Index3Float2Write=true
    Module.Index3Float3Write=false
    Write Parameter Index 0=::NewEnumerator0
    Write Parameter Index 3=::NewEnumerator1
   )
  ]
  rendererClasses [&NiagaraRibbonRendererProperties]
  renderers [
   {rendererClass=&NiagaraRibbonRendererProperties}
   {rendererClass=&NiagaraRibbonRendererProperties, rendererIndex=1}
   {rendererClass=&NiagaraRibbonRendererProperties, rendererIndex=2}
   {rendererClass=&NiagaraRibbonRendererProperties, rendererIndex=3}
   {rendererClass=&NiagaraRibbonRendererProperties, rendererIndex=4}
  ]
 }
 NE_VC_Boost_Sprite_Parent {
  emitterSpawnScript [/SetVariables(
    Emitter.Painted Opacity:NiagaraFloat=&MakeCustomFloatFromBool(Bool=$User.bIsPainted, True Float=0.5, False Float=0.2)
    Emitter.R2 Intensity:NiagaraFloat=&MakeCustomFloatFromBool(Bool=$User.bIsPainted, True Float=HlslExpression() {hlslExpression="User.ColorVariantID == 2 ? 20 : 5"}, False Float=1)
   ) {isSetParameters=true}]
  emitterUpdateScript [
   &EmitterState(
    Inactive Response=::NewEnumerator1
    Life Cycle Mode=::NewEnumerator1
    Loop Behavior=::NewEnumerator1
    Loop Duration Mode=::NewEnumerator1
    UseLoopDelay=false
   )
   &SpawnBurst_Instantaneous(Use Spawn Probability=false, UseLoopCountLimit=false) {enabled=false}
   &VC_FilteredSocketCount()
   &SpawnRate(SpawnRate=120, Use Spawn Probability=false)
  ]
  particleSpawnScript [
   &VC_PaintedVariantsOverride() {enabled=false}
   /SetVariables(Particles.M2 Scale:Vector3f=<1.25 3.1 3.1>) {isSetParameters=true}
   &VC_ExtraBindingColors(
    ExtraColors=2
    R1 Color:LinearColor=&MakeCustomLinearColorFromBool(Bool=$User.bIsPainted, True Color=$User.PaintedColor, False Color=<rgba 1 0.1 0.05 1>)
    R2 Color:LinearColor=&MakeCustomLinearColorFromBool(Bool=$User.bIsPainted, True Color=$User.PaintedColor, False Color=<rgba 1 1 1 1>)
   )
   &VC_PaintedVariantsOverride(
    R1 Black:LinearColor=<rgba 0 0 0 1>
    R1 Crimson:LinearColor=<rgba 0.232 0 0.0063 1>
    R1 DefaultColor=&MakeCustomLinearColorFromBool(Bool=$User.bIsPainted, True Color=$User.PaintedColor, False Color=<rgba 1 0.1 0.05 1>)
    R1 ForestGreen:LinearColor=<rgba 0 0.18 0.0091 1>
    R1 Grey:LinearColor=<rgba 0.1 0.1 0.1 1>
    R1 Lime:LinearColor=<rgba 0.275 0.5 0 1>
    R1 Orange:LinearColor=<rgba 0.43 0.0573 0 1>
    R1 Pink:LinearColor=<rgba 0.44 0.088 0.44 1>
    R1 Purple:LinearColor=<rgba 0.125 0 0.5 1>
    R1 Saffron:LinearColor=<rgba 0.4 0.3667 0 1>
    R1 SkyBlue:LinearColor=<rgba 0 0.175 0.35 1>
    R1 TitaniumWhite:LinearColor=<rgba 0.4 0.4 0.4 1>
    R2 Black:LinearColor=<rgba 0.01 0.01 0.01 1>
    R2 BurntSienna:LinearColor=<rgba 0.13 0.0569 0.0325 1>
    R2 Cobalt:LinearColor=<rgba 0.05 0.0875 0.5 1>
    R2 Crimson:LinearColor=<rgba 0.4 0.05 0.0675 1>
    R2 DefaultColor:LinearColor=&MakeCustomLinearColorFromBool(Bool=$User.bIsPainted, True Color=$User.PaintedColor, False Color=<rgba 1 1 1 1>)
    R2 ForestGreen:LinearColor=<rgba 0.0405 0.15 0.046 1>
    R2 Grey:LinearColor=<rgba 0.1 0.1 0.1 1>
    R2 Lime:LinearColor=<rgba 0.1462 0.25 0.0425 1>
    R2 Orange:LinearColor=<rgba 0.6 0.135 0.042 1>
    R2 Pink:LinearColor=<rgba 0.45 0.1125 0.3273 1>
    R2 Purple:LinearColor=<rgba 0.1438 0.025 0.5 1>
    R2 Saffron:LinearColor=<rgba 0.38 0.3032 0.057 1>
    R2 SkyBlue:LinearColor=<rgba 0.032 0.2236 0.32 1>
    R2 TitaniumWhite:LinearColor=<rgba 0.5 0.5 0.5 1>
    Use R1 Painted Overrides=true
    Use R2 Painted Overrides=true
   ) {module=VC_PaintedVariantsOverride001}
   &InitializeParticle(
    Lifetime=0.1
    Mesh Scale=<5.5 8 8>
    Mesh Scale Mode=::NewEnumerator4
    Sprite Rotation Mode=::NewEnumerator2
    Sprite Size Mode=::NewEnumerator3
    UsePositionOffset=false
    Write Material Random=true
   )
   &PartitionParticles(Number Of Partitions=$Emitter.PortCount)
   &SkeletalMeshLocation(
    Bone / Socket Index=$Particles.PartitionParticles.Partition
    Bone / Socket Sampling Mode=::NewEnumerator5
    Interpolate Mesh Sampling=false
    MaskSpawnGroup=false
    Mesh Sampling Type=::NewEnumerator6
    Sampled Position Offset=&AddVectorToPosition(Position=<0 0 0>, Vector=<-40 0 0>)
    Skeletal Mesh=$System.SkeletalMesh
   )
   &AddVelocity(Velocity=<-1500 0 0>) {enabled=false}
  ]
  particleUpdateScript [
   &ParticleState()
   &ScaleMeshSize(Scale Factor=&VectorFromCurve(VectorCurve=DataInterface() {propertyValues {
       XCurve {defaultValue=1, keys [<at 0 2>,<at 0.4 2.5>,<at 1 3>]}
       YCurve {defaultValue=1, keys [<at 0 1>,<at 0.4 2>,<at 1 0.5>]}
       ZCurve {defaultValue=1, keys [<at 0 1>,<at 0.4 2>,<at 1 0.5>]}
      }}))
   &SolveForcesAndVelocity(Clamp Velocity=false, Limit Acceleration=false)
   /SetVariables(Particles.M2 Scale:Vector3f=&Multiply_Vector(
     A=$Particles.Initial.M2 Scale
     B=&VectorFromCurve(VectorCurve=DataInterface() {propertyValues {
        XCurve {defaultValue=1, keys [<at 0 2>,<at 0.4 2.5>,<at 1 3>]}
        YCurve {defaultValue=1, keys [<at 0 1>,<at 0.4 2>,<at 1 0.5>]}
        ZCurve {defaultValue=1, keys [<at 0 1>,<at 0.4 2>,<at 1 0.5>]}
       }})
    )) {isSetParameters=true}
  ]
  rendererClasses [&NiagaraMeshRendererProperties, &NiagaraSpriteRendererProperties]
  renderers [
   {rendererClass=&NiagaraMeshRendererProperties, rendererIndex=0}
   {rendererClass=&NiagaraMeshRendererProperties, rendererIndex=1}
   {rendererClass=&NiagaraSpriteRendererProperties, rendererIndex=2}
  ]
 }
}
summary {
 systemName=NS_Boost_TurboTechno
 emitters [
  {emitterName=NE_VC_Boost_Ribbon_Parent001, rendererClasses [&NiagaraRibbonRendererProperties]}
  {
   emitterName=NE_VC_Boost_Sprite_Parent
   rendererClasses [&NiagaraMeshRendererProperties, &NiagaraSpriteRendererProperties]
  }
 ]
 userVariables [
  User.bIsFrontendPreview {type=NiagaraBool}
  User.bIsPainted {type=NiagaraBool}
  User.ColorVariantID {type=NiagaraInt32}
  User.PaintedColor {type=LinearColor, description="", defaultValue {struct=LinearColor, value=<rgba 1 1 1 1>}}
  User.PaintedScalar {type=NiagaraFloat, defaultValue=1, description=""}
  User.PortCount {type=NiagaraInt32, defaultValue=1}
 ]
}
